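; Thermal Lance ability that invokes secondary pulses as the firing weapon's overheat climbs.
; On overheat changes while heating and firing, it checks the ratio against hotfix curve thresholds and sends the pulse-fire gameplay event.
; A special Creative check handles timed pulses at full overheat with no cooldowns.
; Firing-tag tasks update or reset predicted pulse timing on the weapon for cosmetics, and ending the ability cleans up.
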(event Ability|EventActivateAbility
  (bind _rangedweaponfiringtag (Variables|Settings|GetRangedWeaponFiringTag))
  ;[0.01906 0.1094 0.01822 0.5] ActivateAbility: Set up binds and variable properties.
  (|CallParentFunction)
  (bind _outfortplayerpawn (CallFunction|GetAvatarasValidFortPlayerPawn)
    (:ValidFortPlayerPawn
      (bind _asynctaskproxy (Ability|Tasks|WaitGameplayTagAdd _rangedweaponfiringtag _outfortplayerpawn)
        (:then
          (Variables|Transient|SetAsyncTask_FiringStart _asynctaskproxy)
          (bind _asynctaskproxy_2 (Ability|Tasks|WaitGameplayTagRemove _rangedweaponfiringtag _outfortplayerpawn)
            (:then
              (Variables|Transient|SetAsyncTask_FiringEnd _asynctaskproxy_2)
              (bind _count (CallFunction|CalculateMaxPulseCount))
              (Variables|Transient|SetMaxPulseCount _count)
              ;;[0.15 0.05028 0.135 0.5] Bind to know when overheat value changes.
              (bind _weapon (Utilities|GetOwnerWeapon)
                (:Valid
                  (Overheating|AssignOnOverheatValueChanged _weapon (EventDispatchers|CreateEvent :!event Custom|HandleOverheatValueChanged))
                  (Variables|Transient|SetCurrentOverheatRatioValue (Math|Float|SafeDivide (Overheat|GetCurrentOverheatValue _weapon :!class /Script/FortniteGame.FortWeaponRanged) (Overheat|GetOverheatingMaxValue _weapon :!class /Script/FortniteGame.FortWeaponRanged)))
                  (Variables|Transient|SetOverheatRatioRate (Math|Float|SafeDivide (Overheat|GetOverheatHeatingValue _weapon) (Overheat|GetOverheatingMaxValue _weapon :!class /Script/FortniteGame.FortWeaponRanged))))))
            (:Removed
              (CallFunction|HandleRangeWeaponFiringStateChanged false))))
        (:Added
          (CallFunction|HandleRangeWeaponFiringStateChanged true))))))

(event Custom|HandleOverheatValueChanged
  (params (CurrentValue :type Float%20%28single-precision%29) (MaxValue :type Float%20%28single-precision%29) (State :type EFortWeaponOverheatState%20Enum) (bIsInCooldown :type Boolean))
  ;[0.0302 0.1419 0.15 0.5] HandleOverheatValueChanged: Check to see if the pulse should be invoked.
  (Variables|Transient|SetPreviousOverheatRatioValue (Variables|Transient|GetCurrentOverheatRatioValue))
  (Variables|Transient|SetCurrentOverheatRatioValue (Math|Float|SafeDivide CurrentValue MaxValue))
  (if (Utilities|Enum|Equal(Enum) State "Heating")
    (bind _weapon (Utilities|GetOwnerWeapon)
      (:Valid
        (if (FortWeapon|IsFiring _weapon)
          (Development|PrintString (Utilities|String|Append "GA: CurrentOverheatRatio: " (Utilities|String|ToString(Float) (Variables|Transient|GetCurrentOverheatRatioValue))) true false "(R=0.000000,G=0.660000,B=1.000000,A=1.000000)" 2.0 "OverheatBasedPulseInvoke_DebugCurrentOverheat")
          (bind _bfirepulse (CallFunction|ShouldWeaponActivatePulse))
          (if _bfirepulse
            (CallFunction|UpdatePulseTimingInfo)
            (else
              (bind (_bretvalue _pulsestartservertime _pulseendservertime) (CallFunction|ShouldFirePulseinCreativeOverheatNoCooldownState))
              (if _bretvalue
                (Pulse|SetPulseTimingInfo _weapon (Utilities|Struct|MakeThermalLancePulseTimingInfoStruct _pulsestartservertime _pulseendservertime)))))
          (Ability|SendGameplayEventtoActor (Ability|GetAvatarActorfromActorInfo) "(TagName=\"ThermalLance.Event.Pulse.Fire\")") ; Fires the pulse
          (Variables|Transient|SetLastPulseFireServerTime (Fortnite|Sync|Time|GetServerWorldTimeSeconds)))))))

(event Custom|HandleRangeWeaponFiringStateChanged
  (params (bFiring :type Boolean))
  ;[0.15 0.06166 0.141 0.5] HandleRangeWeaponFiringStateChanged: When the player starts firing their beam, update the pulse timing info. When the player stops firing their beam, reset their pulse timing info.
  (if bFiring
    (CallFunction|UpdatePulseTimingInfo)
    (else
      (CallFunction|ResetPulseTimingInfo))))

(event Ability|EventOnEndAbility
  (params (bWasCancelled :type Boolean))
  ;[0.15 0.07248 0.02286 0.5] EndPlay: Clean up.
  (Utilities|FlowControl|Sequence
    (:then_0
      (GameplayTasks|EndTask (Variables|Transient|GetAsyncTask_FiringStart))
      (CallFunction|ResetPulseTimingInfo)
      (bind _weapon (Utilities|GetOwnerWeapon)
        (:Valid
          (Overheating|UnbindEventfromOnOverheatValueChanged _weapon (EventDispatchers|CreateEvent :!event Custom|HandleOverheatValueChanged)))))
    (:then_1
      (|CallParentFunction bWasCancelled))))

(fn CalculateMaxPulseCount
  ;[0.05293 0.1004 0.15 0.5] Based on the hotfix row, calculate how many pulses are desired to be fired in the span of the overheat progression.
  (while (>= (ScalableFloat|GetValueatLevel (Variables|Hotfix|GetHotfix_PulseFireOverheatCadence) (Math|Conversions|ToFloat(Integer) (LocalVariables|GetIndex))) 0)
    (Math|Integer|IncrementInt (LocalVariables|GetIndex)))
  (return (LocalVariables|GetIndex)))

(fn ResetPulseTimingInfo
  ;[0 0.15 0.1416 0.5] Reset the pulse timing info to null values.
  (bind _weapon (Utilities|GetOwnerWeapon)
    (:Valid
      (Pulse|SetPulseTimingInfo _weapon (Utilities|Struct|MakeThermalLancePulseTimingInfoStruct 0.0)))))

(fn ShouldFirePulseInCreativeOverheatNoCooldownState
  (bind _maxpulsecount (Variables|Transient|GetMaxPulseCount))
  (bind _returnvalue (- _maxpulsecount 2))
  (bind _returnvalue_2 (Math|Conversions|ToFloat(Integer) _returnvalue))
  (bind _returnvalue_3 (ScalableFloat|GetValueatLevel (Variables|Hotfix|GetHotfix_PulseFireOverheatCadence) _returnvalue_2))
  (bind _maxpulsecount_2 (Variables|Transient|GetMaxPulseCount))
  (bind _returnvalue_4 (- _maxpulsecount_2 1))
  (bind _returnvalue_5 (Math|Conversions|ToFloat(Integer) _returnvalue_4))
  (bind _hotfix_pulsefireoverheatcadence (Variables|Hotfix|GetHotfix_PulseFireOverheatCadence))
  (bind _returnvalue_6 (ScalableFloat|GetValueatLevel _hotfix_pulsefireoverheatcadence _returnvalue_5))
  (bind _returnvalue_7 (- _returnvalue_6 _returnvalue_3))
  (bind _returnvalue_8 (Math|Float|MapRangeClamped _returnvalue_7 0.0 _returnvalue_6 0.0 1.0))
  ;[0.1073 0.04053 0.15 0.5] Returns if the weapon should fire a pulse while being used in Creative being overheated with no cooldowns ON setting enabled.
  (Utilities|FlowControl|Sequence
    (:then_0
      (if (> (ScalableFloat|GetCachedScalableFloatValue (Variables|Hotfix|GetHotfix_CreativePulseFireNoCooldownBehavior)) 0)
        (if (and (== (Variables|Transient|GetCurrentOverheatRatioValue) (Variables|Transient|GetPreviousOverheatRatioValue)) (== (Variables|Transient|GetCurrentOverheatRatioValue) 1.0))
          (bind _outfortplayerpawn (CallFunction|GetAvatarasValidFortPlayerPawn)
            (:ValidFortPlayerPawn
              (bind _returnvalue_9 (Fortnite|Athena|GameState|GetExperienceSettingsComponentfromGameState))
              (Utilities|IsValid _returnvalue_9
                (:Is%20Valid
                  (if (ScalableFloat|GetValueasBoolwithContextParam _outfortplayerpawn (Variables|/Script/FortniteGame.FortExperienceSettingsComponent|GetNoCooldowns _returnvalue_9))
                    (bind _weapon (Utilities|GetOwnerWeapon)
                      (:Valid
                        (if (> (- (Fortnite|Sync|Time|GetServerWorldTimeSeconds) (Variables|Transient|GetLastPulseFireServerTime)) (* _returnvalue_8 (Math|Float|SafeDivide (Overheat|GetOverheatingMaxValue _weapon :!class /Script/FortniteGame.FortWeaponRanged) (* (Overheat|GetOverheatHeatingValue _weapon) (FortWeapon|GetFiringRateFastandNotExact _weapon)))))
                          (return true (Fortnite|Sync|Time|GetServerWorldTimeSeconds) (+ (Fortnite|Sync|Time|GetServerWorldTimeSeconds) (* _returnvalue_8 (Math|Float|SafeDivide (Overheat|GetOverheatingMaxValue _weapon :!class /Script/FortniteGame.FortWeaponRanged) (* (Overheat|GetOverheatHeatingValue _weapon) (FortWeapon|GetFiringRateFastandNotExact _weapon)))))))))))))))))
    (:then_1
      (return false 0.0 0.0))))

(fn ShouldWeaponActivatePulse
  (bind _index (LocalVariables|GetIndex))
  (bind _returnvalue (Math|Conversions|ToFloat(Integer) _index))
  (bind _hotfix_pulsefireoverheatcadence (Variables|Hotfix|GetHotfix_PulseFireOverheatCadence))
  (bind _returnvalue_2 (ScalableFloat|GetValueatLevel _hotfix_pulsefireoverheatcadence _returnvalue))
  ;[0.05652 0.128 0.2066 0.5] ShouldWeaponActivatePulse: Check to see if the current overheat progression has crossed a threshold to fire a pulse.
  (while (>= _returnvalue_2 0)
    (bind _output_get (LocalVariables|SetCurrentCurveTableValue _returnvalue_2))
    (Math|Integer|IncrementInt (LocalVariables|GetIndex)
      (:%20%20
        (if (and (< (Variables|Transient|GetPreviousOverheatRatioValue) _output_get) (>= (Variables|Transient|GetCurrentOverheatRatioValue) _output_get))
          (return true)))))
  (return false))

(fn UpdatePulseTimingInfo
  ;[0.03123 0.15 0.143 0.5] UpdatePulseTimingInfo: Update the information that is used by the weapon to anticipate future pulses for cosmetic reasons.
  (bind _weapon (Utilities|GetOwnerWeapon)
    (:Valid
      (bind _output_get (LocalVariables|SetWeapon _weapon))
      (Utilities|FlowControl|Sequence
        (:then_0
          (bind _index (Utilities|FlowControl|ForLoopwithBreak 0 (- (Variables|Transient|GetMaxPulseCount) 2))
            (:LoopBody
              (LocalVariables|SetOverheatRatioValue_PrimaryIndex (ScalableFloat|GetValueatLevel (Variables|Hotfix|GetHotfix_PulseFireOverheatCadence) (Math|Conversions|ToFloat(Integer) _index)))
              (LocalVariables|SetOverheatRatioValue_SecondaryIndex (ScalableFloat|GetValueatLevel (Variables|Hotfix|GetHotfix_PulseFireOverheatCadence) (Math|Conversions|ToFloat(Integer) (+ _index 1))))
              (if (and (>= (Variables|Transient|GetCurrentOverheatRatioValue) (LocalVariables|GetOverheatRatioValue_PrimaryIndex)) (< (Variables|Transient|GetCurrentOverheatRatioValue) (LocalVariables|GetOverheatRatioValue_SecondaryIndex)))
                (LocalVariables|SetRetTimingInfo (Utilities|Struct|MakeThermalLancePulseTimingInfoStruct (- (Fortnite|Sync|Time|GetServerWorldTimeSeconds) 0) (+ (Fortnite|Sync|Time|GetServerWorldTimeSeconds) (Utilities|Operators|Multiply (Overheat|GetOverheatingMaxValue (LocalVariables|GetWeapon) :!class /Script/FortniteGame.FortWeaponRanged) (- (LocalVariables|GetOverheatRatioValue_SecondaryIndex) (Variables|Transient|GetCurrentOverheatRatioValue)) (Variables|Transient|GetOverheatRatioRate)))))))))
        (:then_1
          (Pulse|SetPulseTimingInfo _output_get (LocalVariables|GetRetTimingInfo)))))))
