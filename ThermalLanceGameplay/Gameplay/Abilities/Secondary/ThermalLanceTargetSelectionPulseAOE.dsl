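; Targeting selection for the Thermal Lance secondary ability's pulse AOE, centering the area query on the weapon's latest pulse hit.
; Its GetSourceLocation override takes the source context from the targeting subsystem, casts it to a Fort player pawn and reads the hit location from that pawn's current B_Weapon_ThermalLance.
; A Sequence node provides the fallback to the parent implementation when the subsystem is invalid or a cast fails.
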(fn GetSourceLocation
  (params (TargetingHandle :type Targeting%20Request%20Handle%20Structure%20%28by%20ref%29))
  (bind _returnvalue (Targeting|GetTargetingSubsystem TargetingHandle))
  (Utilities|FlowControl|Sequence
    (:then_0
      (Utilities|IsValid _returnvalue
        (:Is%20Valid
          (bind _asfort_player_pawn (Utilities|Casting|CastToFortPlayerPawn (Utilities|Struct|BreakTargetingSourceContext (TargetingSystem|TargetingTypes|GetTargetingSourceContext _returnvalue TargetingHandle))))
          (bind _asb_weapon_thermal_lance (Utilities|Casting|CastToB_Weapon_ThermalLance (Weapon|GetCurrentWeapon _asfort_player_pawn)))
          (return (Pulse|GetLatestPulseHitLocation _asb_weapon_thermal_lance)))))
    (:then_1
      (bind _returnvalue_2 (|CallParentFunction TargetingHandle))
      (return _returnvalue_2))))
